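# Material function in the UI Modular Button library that exposes state transition parameters for modular buttons with a Select state.
# Each scalar parameter in the "State transitions" group feeds its own output unchanged, covering hovered, focused, pressed and disabled, plus selected idle, hovered, focused, pressed and disabled.
# Values are meant as 0 to 1 transition amounts that can drive curves for the desired visual effect.
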MaterialFunction /FortUILibrary/Systems/ModularButton/MaterialFunctions/MF_UIKit_ButtonStatesSelectable_Params
graph {
 properties {
  description="Exposes parameters for Modular Button Base states transition with Select state as well. Provides transition value ( 0 - 1 ) and could be used as input for curves to achieve desired changing effect."
  libraryCategoriesText [UI Modular Button]
 }
 nodes [
  FunctionOutput(None=ScalarParameter() {group=State transitions, parameterName=Trans_Hovered, sliderMax=1, sortPriority=1}) {outputName=Trans_Hovered}
  FunctionOutput(None=ScalarParameter() {group=State transitions, parameterName=Trans_Focused, sliderMax=1, sortPriority=2}) {outputName=Trans_Focused, sortPriority=1}
  FunctionOutput(None=ScalarParameter() {group=State transitions, parameterName=Trans_SelectedHovered, sliderMax=1, sortPriority=11}) {outputName=Trans_SelectedHovered, sortPriority=5}
  FunctionOutput(None=ScalarParameter() {group=State transitions, parameterName=Trans_Pressed, sliderMax=1, sortPriority=3}) {outputName=Trans_Pressed, sortPriority=2}
  FunctionOutput(None=ScalarParameter() {group=State transitions, parameterName=Trans_Disabled, sliderMax=1, sortPriority=4}) {outputName=Trans_Disabled, sortPriority=3}
  FunctionOutput(None=ScalarParameter() {group=State transitions, parameterName=Trans_SelectedFocused, sliderMax=1, sortPriority=12}) {outputName=Trans_SelectedFocused, sortPriority=6}
  FunctionOutput(None=ScalarParameter() {group=State transitions, parameterName=Trans_SelectedPressed, sliderMax=1, sortPriority=13}) {outputName=Trans_SelectedPressed, sortPriority=7}
  FunctionOutput(None=ScalarParameter() {group=State transitions, parameterName=Trans_SelectedDisabled, sliderMax=1, sortPriority=14}) {outputName=Trans_SelectedDisabled, sortPriority=8}
  FunctionOutput(None=ScalarParameter() {group=State transitions, parameterName=Trans_SelectedIdle, sliderMax=1, sortPriority=10}) {outputName=Trans_SelectedIdle, sortPriority=4}
 ]
}
