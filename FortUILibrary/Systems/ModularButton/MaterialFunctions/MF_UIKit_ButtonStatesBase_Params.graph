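# Material function for the UI kit's Modular Button base that exposes its state transition values to materials, listed in the library under UI Modular Button.
# Four scalar parameters in the State transitions group, for Hovered, Focused, Pressed and Disabled, each pass straight to a matching output.
# Each is a 0 to 1 transition value that can feed curves to shape how the button's look changes between states.
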MaterialFunction /FortUILibrary/Systems/ModularButton/MaterialFunctions/MF_UIKit_ButtonStatesBase_Params
graph {
 properties {
  bExposeToLibrary=true
  description="Exposes parameters for Modular Button Base states transition. Provides transition value ( 0 - 1 ) and could be used as input for curves to achieve desired changing effect."
  libraryCategoriesText [UI Modular Button]
 }
 nodes [
  FunctionOutput(None=ScalarParameter() {group=State transitions, parameterName=Trans_Hovered, sliderMax=1, sortPriority=1}) {outputName=Trans_Hovered}
  FunctionOutput(None=ScalarParameter() {group=State transitions, parameterName=Trans_Focused, sliderMax=1, sortPriority=2}) {outputName=Trans_Focused, sortPriority=1}
  FunctionOutput(None=ScalarParameter() {group=State transitions, parameterName=Trans_Pressed, sliderMax=1, sortPriority=3}) {outputName=Trans_Pressed, sortPriority=2}
  FunctionOutput(None=ScalarParameter() {group=State transitions, parameterName=Trans_Disabled, sliderMax=1, sortPriority=4}) {outputName=Trans_Disabled, sortPriority=3}
 ]
}
